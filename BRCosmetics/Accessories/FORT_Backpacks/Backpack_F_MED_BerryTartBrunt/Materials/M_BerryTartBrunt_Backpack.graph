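# Character-parent material for the BerryTartBrunt back bling, adding an audio-reactive animated screen.
# Custom UV1 remapping places flipbook idle and emote animations and two alternating rhythm icons, with bump offset adding depth to the animations and screen shadow mask.
# MF_AudioReactivity amplitude drives a reactivity alpha, with switches for disable and debug override.
# Masks composite the screen into base color and emissive between the PreFX and PostFX parent functions.
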Material /BRCosmetics/Accessories/FORT_Backpacks/Backpack_F_MED_BerryTartBrunt/Materials/M_BerryTartBrunt_Backpack
graph {
 properties {
  bHasPixelAnimation=true
  bUseMaterialAttributes=true
  bUsedWithMeshDeformer=true
  bUsedWithMorphTargets=true
  bUsedWithSkeletalMesh=true
  pixelDepthOffsetMode=PDOM_Legacy
  shadingModel=MSM_FromMaterialExpression
  usedShadingModels=Default Lit | Subsurface | Clear Coat | Subsurface Profile | Cloth
 }
 nodes [
  n1 = MaterialFunctionCall() {materialFunction=&MF_CharacterParent_2023_PreFX}
  n2 = GetMaterialAttributes(None=@n1.Result) {attributeGetTypes [#1, #2, #3]}
  n3 = LinearInterpolate(
   A=@n2.Base Color
   B=VectorParameter() {group=Screen_Reactivity, parameterName=ReactivityBackGroundColor, defaultValue=<rgba 0 0.7655 1 1>}.RGB
   Alpha=NamedRerouteUsage()
  )
  n4 = ScalarParameter() {defaultValue=0.7, group=Screen, parameterName=ShadowMax}
  n5 = TextureCoordinate()
  n6 = ScalarParameter() {defaultValue=-0.2, group=Screen, parameterName=BumpOffsetHeight}
  n7 = TextureObjectParameter() {
   group=Screen
   parameterName=ScreenMask
   samplerType=SAMPLERTYPE_Masks
   texture=/BRCosmetics/Accessories/FORT_Backpacks/Backpack_F_MED_BerryTartBrunt/Texture/T_BerryTartBrunt_Additional_FX
  }
  n8 = TextureSample(UVs=BumpOffset(Coordinate=@n5, Height=@n6), Tex=@n7) {
   samplerSource=SSM_Wrap_WorldGroupSettings
   samplerType=SAMPLERTYPE_Masks
   texture=&T_BerryTartBrunt_Additional_FX
  }
  n9 = LinearInterpolate(
   A=ScalarParameter() {defaultValue=0.2, group=Screen, parameterName=ShadowColorMulti}
   Alpha=NamedRerouteDeclaration(None=SmoothStep(Max=@n4, Value=@n8.G) {constMax=0.3}) {name=ShadowArea, nodeColor=<rgba 0 1 0.3906 1>}
  )
  n10 = LinearInterpolate(A=NamedRerouteUsage(), B=NamedRerouteUsage(), Alpha=NamedRerouteUsage())
  n11 = ScalarParameter() {defaultValue=0.5, group=Screen_Reactivity, parameterName=ScreenBaseColor_Intensity}
  n12 = LinearInterpolate(A=NamedRerouteUsage(), B=NamedRerouteUsage(), Alpha=NamedRerouteUsage())
  n13 = LinearInterpolate(
   A=LinearInterpolate(A=@n2.Base Color, B=Multiply(A=@n3, B=@n9), Alpha=NamedRerouteUsage())
   B=Desaturation(
    None=Multiply(A=@n10, B=@n11)
    Fraction=ScalarParameter() {defaultValue=0.2, group=Screen_Reactivity, parameterName=ScreenBaseColor_Desaturation}
   )
   Alpha=@n12
  )
  n14 = NamedRerouteUsage()
  n15 = NamedRerouteUsage()
  MaterialFunctionCall(
   Material=SetMaterialAttributes!(
    MaterialAttributes=@n2.MaterialAttributes
    Base Color=LinearInterpolate(A=@n13, Alpha=@n14) {constB=0}
    Emissive Color=Add(A=@n2.Emissive Color, B=@n15)
    Specular=@n2.MaterialAttributes
   ) {attributeSetTypes [#1, #2, #3]}
   M=@n1.M
   Pre Skinned Local Position=@n1.M
   Distance Blend=@n1.M
   Fresnel=@n1.M
  ) {materialFunction=&MF_CharacterParent_2023_PostFX}
  n16 = Power(Base=NamedRerouteUsage(), Exp=ScalarParameter() {defaultValue=2, group=Screen_Reactivity, parameterName=AmplitudePower}) {constExponent=8}
  Multiply(A=Step(X=@n16) {constY=0.3}, B=@n16)
  n17 = Time()
  n18 = ScalarParameter() {defaultValue=0.5, group=Screen_Reactivity, parameterName=IdleAnimationSpeed}
  n19 = Constant() {r=2}
  n20 = TextureObjectParameter() {
   group=Screen_Reactivity
   parameterName=IdleAnimation
   texture=/BRCosmetics/Accessories/FORT_Backpacks/Backpack_F_MED_BerryTartBrunt/Texture/T_BerryIdle
  }
  n21 = VectorParameter() {group=Screen_Reactivity, parameterName=CenterUV01, channelNames {a=ScaleY, b=ScaleX, g=CenterY, r=CenterX}, defaultValue=<rgba 0.5 0.5 1 1>}
  MaterialExpressionAppendVector_2 = AppendVector(A=@n21.CenterX, B=@n21.CenterX)
  MaterialExpressionAppendVector_3 = AppendVector(A=@n21.ScaleX, B=@n21.ScaleX)
  MaterialExpressionTextureCoordinate_1 = TextureCoordinate() {coordinateIndex=1}
  n22 = Custom!(UVPoint=@MaterialExpressionAppendVector_2, UVScale=@MaterialExpressionAppendVector_3, OriginalUV=@MaterialExpressionTextureCoordinate_1) {
   code="float2 DistUV =OriginalUV-UVPoint;\r\nDistUV = DistUV/UVScale + float2(0.5,0.5);\r\n\r\nreturn DistUV;"
   description=CopyToPointUV
   outputType=CMOT_Float2
   inputs [
    {
     inputName=UVPoint
     input {expression=&self:MaterialExpressionAppendVector_2}
    }
    {
     inputName=UVScale
     input {expression=&self:MaterialExpressionAppendVector_3}
    }
    {
     inputName=OriginalUV
     input {expression=&self:MaterialExpressionTextureCoordinate_1}
    }
   ]
  }
  n23 = BumpOffset(Coordinate=@n22, Height=ScalarParameter() {defaultValue=-0.4, group=Screen, parameterName=BumpOffsetHeight_UV02})
  n24 = MaterialFunctionCall("Animation  Phase (0-1)"=Frac(None=Multiply(A=@n17, B=@n18)), Number of Rows=@n19, Number of Columns=@n19, Texture=@n20, UVs=@n23) {materialFunction=&FlipBook}
  NamedRerouteDeclaration(None=@n24.Result) {name=IdleAnimation, nodeColor=<rgba 1 0.7734 0 1>}
  n25 = MaterialFunctionCall!(UV_Coord=@n23) {materialFunction=&/MF_UV01_Mask}
  NamedRerouteDeclaration(None=Multiply(A=@n24.Alpha, B=@n25.Mask)) {name=IdleAnimationAlpha, nodeColor=<rgba 0 1 0.1328 1>}
  NamedRerouteDeclaration(None=@n12) {name=AlphaComp, nodeColor=<rgba 1 0 0.3516 1>}
  n26 = AppendVector(A=Constant(), B=ScalarParameter() {defaultValue=0.1, group=Screen_Reactivity, parameterName=IconOffset})
  n27 = Divide(A=NamedRerouteUsage()) {constB=4}
  n28 = Frac(None=Multiply(A=Time(), B=Multiply(A=@n27) {constB=9}))
  n29 = VectorParameter() {group=Screen_Reactivity, parameterName=IconUV01, channelNames {a=ScaleY, b=ScaleX, g=CenterY, r=CenterX}, defaultValue=<rgba 0.5 0.5 1 1>}
  MaterialExpressionAdd_0 = Add(A=Multiply(A=@n26, B=Step(X=@n28) {constY=0.5}), B=AppendVector(A=@n29.CenterX, B=@n29.CenterX))
  MaterialExpressionAppendVector_1 = AppendVector(A=@n29.ScaleX, B=@n29.ScaleX)
  MaterialExpressionTextureCoordinate_2 = TextureCoordinate() {coordinateIndex=1}
  n30 = Custom(UVPoint=@MaterialExpressionAdd_0, UVScale=@MaterialExpressionAppendVector_1, OriginalUV=@MaterialExpressionTextureCoordinate_2) {
   code="float2 DistUV =OriginalUV-UVPoint;\r\nDistUV = DistUV/UVScale + float2(0.5,0.5);\r\n\r\nreturn DistUV;"
   description=CopyToPointUV
   outputType=CMOT_Float2
   inputs [
    {
     inputName=UVPoint
     input {expression=&self:MaterialExpressionAdd_0}
    }
    {
     inputName=UVScale
     input {expression=&self:MaterialExpressionAppendVector_1}
    }
    {
     inputName=OriginalUV
     input {expression=&self:MaterialExpressionTextureCoordinate_2}
    }
   ]
  }
  n31 = TextureSampleParameter2D(UVs=@n30) {
   group=Screen_Reactivity
   parameterName=Icons
   samplerType=SAMPLERTYPE_Masks
   texture=/BRCosmetics/Accessories/FORT_Backpacks/Backpack_F_MED_BerryTartBrunt/Texture/T_RythmIcon
  }
  n32 = Multiply(
   A=Multiply(
    A=@n31.R
    B=MaterialFunctionCall(UV_Coord=@n30) {materialFunction=&MF_UV01_Mask}.Mask
   )
   B=NamedRerouteUsage()
  )
  n33 = Divide(A=NamedRerouteUsage()) {constB=4}
  n34 = Frac(None=Multiply(A=Time(), B=Multiply(A=@n33) {constB=9}))
  n35 = VectorParameter() {group=Screen_Reactivity, parameterName=IconUV02, channelNames {a=ScaleY, b=ScaleX, g=CenterY, r=CenterX}, defaultValue=<rgba 0.5 0.5 1 1>}
  MaterialExpressionAdd_3 = Add(A=Multiply(A=@n26, B=Step(Y=@n34) {constX=0.5}), B=AppendVector(A=@n35.CenterX, B=@n35.CenterX))
  MaterialExpressionAppendVector_6 = AppendVector(A=@n35.ScaleX, B=@n35.ScaleX)
  MaterialExpressionTextureCoordinate_3 = TextureCoordinate() {coordinateIndex=1}
  n36 = Custom(UVPoint=@MaterialExpressionAdd_3, UVScale=@MaterialExpressionAppendVector_6, OriginalUV=@MaterialExpressionTextureCoordinate_3) {
   code="float2 DistUV =OriginalUV-UVPoint;\r\nDistUV = DistUV/UVScale + float2(0.5,0.5);\r\n\r\nreturn DistUV;"
   description=CopyToPointUV
   outputType=CMOT_Float2
   inputs [
    {
     inputName=UVPoint
     input {expression=&self:MaterialExpressionAdd_3}
    }
    {
     inputName=UVScale
     input {expression=&self:MaterialExpressionAppendVector_6}
    }
    {
     inputName=OriginalUV
     input {expression=&self:MaterialExpressionTextureCoordinate_3}
    }
   ]
  }
  n37 = TextureSampleParameter2D(UVs=@n36) {
   group=Screen_Reactivity
   parameterName=Icons
   samplerSource=SSM_Wrap_WorldGroupSettings
   samplerType=SAMPLERTYPE_Masks
   texture=&T_RythmIcon
  }
  n38 = Multiply(
   A=Multiply(
    A=@n37.G
    B=MaterialFunctionCall(UV_Coord=@n36) {materialFunction=&MF_UV01_Mask}.Mask
   )
   B=NamedRerouteUsage()
  )
  NamedRerouteDeclaration(None=Add(
    A=NamedRerouteDeclaration(None=@n32) {name=IconAlpha01, nodeColor=<rgba 0.4297 0 1 1>}
    B=NamedRerouteDeclaration(None=@n38) {name=IconAlpha02, nodeColor=<rgba 0.4297 0 1 1>}
   )) {name=Icon_Alpha, nodeColor=<rgba 0.1016 0 1 1>}
  n39 = MaterialFunctionCall() {materialFunction=/Game/Packages/Fortress_SharedMaterials/Base_Material_Functions/BaseCharacterShaderFunctions/Audio/MF_AudioReactivity}
  NamedRerouteDeclaration(None=@n39.Amplitude) {name=Amplitude, nodeColor=<rgba 0.3594 0 1 1>}
  n40 = VectorParameter() {group=Screen_Reactivity, parameterName=LineTint_BUTTON, channelNames {a=Intensity, b="", g="", r=""}, defaultValue=<rgba 1 0 0.6632 1>}
  n41 = TextureSampleParameter2D() {
   group=Screen_Reactivity
   parameterName=ReactivityEmissive
   samplerType=SAMPLERTYPE_Masks
   texture=/BRCosmetics/Accessories/FORT_Backpacks/Backpack_F_MED_BerryTartBrunt/Texture/T_ReactivityMask
  }
  n42 = Multiply(
   A=LinearInterpolate(A=ScalarParameter() {defaultValue=0.3, group=Screen_Reactivity, parameterName=AmplitudeMin}, Alpha=@n16) {constA=0.5}
   B=Multiply(A=Multiply(A=@n40, B=@n40), B=@n41.R)
  )
  n43 = VectorParameter() {group=Screen_Reactivity, parameterName=LineTint_HIGH, channelNames {a=Intensity, b="", g="", r=""}, defaultValue=<rgba 1 0 0.6632 1>}
  n44 = Multiply(A=Power(Base=@n41.G, Exp=ScalarParameter() {defaultValue=3, group=Screen_Reactivity, parameterName=Area_Pow}), B=Multiply(A=@n43, B=@n43))
  NamedRerouteDeclaration(None=Multiply(A=Add(A=@n42, B=@n44), B=NamedRerouteUsage())) {name=EM_Music, nodeColor=<rgba 1 0.7266 0 1>}
  NamedRerouteDeclaration(None=@n39.Spectrum) {name=Spectrum, nodeColor=<rgba 0 1 0.4375 1>}
  NamedRerouteDeclaration(None=@n41.B) {name=SpectrumArea, nodeColor=<rgba 1 0.3516 0 1>}
  NamedRerouteDeclaration(None=@n6) {name=OffsetHeight, nodeColor=<rgba 0.1016 0 1 1>}
  n45 = ScalarParameter() {group=Screen_Reactivity, parameterName=Reactivity_Alpha}
  n46 = SmoothStep(Value=@n39.Amplitude) {constMax=0.2}
  NamedRerouteDeclaration(None=StaticSwitchParameter(True=StaticSwitchParameter(True=@n45, False=@n46) {group=Screen_Reactivity, parameterName=DebugReactivity}, False=Constant()) {defaultValue=true, group=Screen_Reactivity, parameterName=EnableReactivity}) {name=ReactivityAlpha, nodeColor=<rgba 0 1 0.9766 1>}
  n47 = VectorParameter() {group=Screen_Reactivity, parameterName=ReactivityBackGroundColor, defaultValue=<rgba 0 0.7655 1 1>}
  n48 = LinearInterpolate(B=LinearInterpolate(A=Multiply(A=@n47.RGB) {constB=0.5}, B=@n47.RGB, Alpha=NamedRerouteUsage()), Alpha=NamedRerouteUsage())
  n49 = Multiply(
   A=NamedRerouteUsage()
   B=LinearInterpolate(
    A=ScalarParameter() {defaultValue=0.1, group=Screen_Reactivity, parameterName=ScreenEM_Intensity}
    B=ScalarParameter() {defaultValue=0.3, group=Screen_Reactivity, parameterName=ScreenEM_Intensity_React}
    Alpha=NamedRerouteUsage()
   )
  )
  n50 = NamedRerouteUsage()
  n51 = Fresnel()
  n52 = Multiply(A=Add(A=LinearInterpolate(A=@n48, B=@n49, Alpha=@n50), B=Multiply(A=@n51) {constB=0.1}), B=NamedRerouteUsage())
  n53 = NamedRerouteUsage()
  NamedRerouteDeclaration(None=Add(A=LinearInterpolate(A=@n52, Alpha=@n53) {constB=0}, B=NamedRerouteUsage())) {name=AdditionalEM, nodeColor=<rgba 1 0.6563 0 1>}
  n54 = TextureSample(UVs=@n5, Tex=@n7) {
   samplerSource=SSM_Wrap_WorldGroupSettings
   samplerType=SAMPLERTYPE_Masks
   texture=&T_BerryTartBrunt_Additional_FX
  }
  NamedRerouteDeclaration(None=@n54.R) {name=AreaMask, nodeColor=<rgba 0 1 0.3203 1>}
  n55 = NamedRerouteUsage()
  n56 = Multiply(A=Time(), B=Multiply(A=Divide(A=@n55) {constB=9}) {constB=4})
  n57 = Constant() {r=4}
  n58 = TextureObjectParameter() {
   group=Screen_Reactivity
   parameterName=EmoteAnimation
   texture=/BRCosmetics/Accessories/FORT_Backpacks/Backpack_F_MED_BerryTartBrunt/Texture/T_BerryEmote01
  }
  n59 = MaterialFunctionCall(
   "Animation  Phase (0-1)"=LinearInterpolate(B=Divide() {constA=9, constB=16}, Alpha=Frac(None=@n56))
   Number of Rows=@n57
   Number of Columns=@n57
   Texture=@n58
   UVs=@n23
  ) {materialFunction=&FlipBook}
  NamedRerouteDeclaration(None=Multiply(A=@n25.Mask, B=@n59.Alpha)) {name=EmoteAnimationAlpha, nodeColor=<rgba 0 1 0.1328 1>}
  NamedRerouteDeclaration(None=@n59.Result) {name=EmoteAnimation, nodeColor=<rgba 1 0.6094 0 1>}
  NamedRerouteDeclaration(None=@n18) {name=IdleSpeed, nodeColor=<rgba 0.2891 0 1 1>}
  NamedRerouteDeclaration(None=@n10) {name=AnimationComp, nodeColor=<rgba 0.0312 1 0 1>}
 ]
}
